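# Control Rig for the DistantEchoPro male character's face accessory part, adjusting hair bones procedurally as the head moves.
# A spherical pose reader measures head orientation against neck_01, with debug drawing enabled, and its output weights rotation offsets on the back-right and back-center hair master bones.
# The back-left hair master and front-left strand receive fixed rotation offsets at full weight, independent of the reader.
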ControlRigBlueprint /BRCosmetics/Characters/Player/Male/Medium/Bodies/M_MED_DistantEchoPro/Meshes/Parts/M_MED_DistantEchoPro_FaceAcc_CtrlRig
graphs [
 RigVMFunctionLibrary {links [], nodes []}
 RigVMModel {nodes [
   RigUnit_BeginExecution = &RigUnit_BeginExecution()
   n1 = &RigUnit_SphericalPoseReader(
    ActiveRegionScaleFactors=(PositiveWidth=1,NegativeWidth=0.2,PositiveHeight=1,NegativeHeight=1)
    ActiveRegionSize=0.2
    Debug=(bDrawDebug=True,DebugScale=25,DebugSegments=20,DebugThickness=0.25)
    DriverAxis=(X=1,Y=0,Z=0)
    DriverItem=(Type=Bone,Name="head")
    FalloffRegionScaleFactors=(PositiveWidth=0.5,NegativeWidth=0.2,PositiveHeight=0.5,NegativeHeight=0.5)
    FalloffSize=0.45
    OptionalParentItem=(Type=Bone,Name="neck_01")
    RotationOffset=(X=-90,Y=0,Z=30)
    ExecutePin=@RigUnit_BeginExecution.ExecutePin
   )
   > &RigUnit_ModifyTransforms(
    ItemToModify=((Item=(Type=Bone,Name="dyn_hair_bk_r_MSTR"),Transform=(Rotation=(X=0,Y=0,Z=0,W=1),Translation=(X=0,Y=0,Z=0),Scale3D=(X=1,Y=1,Z=1))))
    ItemToModify.0.Transform.Rotation=&RigVMFunction_MathQuaternionFromEuler(Euler=(X=0,Y=0,Z=-40)).Result
    Weight=@n1.OutputParam
   )
   > &RigUnit_ModifyTransforms(
    ItemToModify=((Item=(Type=Bone,Name="dyn_hair_bk_l_MSTR"),Transform=(Rotation=(X=0,Y=0,Z=0,W=1),Translation=(X=0,Y=0,Z=0),Scale3D=(X=1,Y=1,Z=1))))
    Weight=1
    ItemToModify.0.Transform.Rotation=&RigVMFunction_MathQuaternionFromEuler(Euler=(X=0,Y=0,Z=-10)).Result
   )
   > &RigUnit_ModifyTransforms(
    ItemToModify=((Item=(Type=Bone,Name="dyn_hair_bk_MSTR"),Transform=(Rotation=(X=0,Y=0,Z=0,W=1),Translation=(X=0,Y=0,Z=0),Scale3D=(X=1,Y=1,Z=1))))
    ItemToModify.0.Transform.Rotation=&RigVMFunction_MathQuaternionFromEuler(Euler=(X=0,Y=0,Z=-20)).Result
    Weight=@n1.OutputParam
   )
   > &RigUnit_ModifyTransforms(
    ItemToModify=((Item=(Type=Bone,Name="dyn_strand_fr_l"),Transform=(Rotation=(X=0,Y=0,Z=0,W=1),Translation=(X=0,Y=0,Z=0),Scale3D=(X=1,Y=1,Z=1))))
    Weight=1
    ItemToModify.0.Transform.Rotation=&RigVMFunction_MathQuaternionFromEuler(Euler=(X=0,Y=10,Z=15)).Result
   )
  ]}
]
hierarchy [
 root {loc=<0 0 0>, rot=<0 0 0 1>}
 pelvis {parent=root, loc=<0 -0.7071 93.1191>, rot=<0.0127 0.707 -0.0127 -0.707>}
 spine_01 {parent=pelvis, loc=<6.7494 -0.6388 0>, rot=<0 0 0.034 -0.9994>}
 spine_02 {parent=spine_01, loc=<7.2518 -0.035 0>, rot=<0 0 0.0238 -0.9997>}
 spine_03 {parent=spine_02, loc=<6.0104 0 0>, rot=<0 0 0.0149 -0.9999>}
 spine_04 {parent=spine_03, loc=<8.1024 0 0>, rot=<0 0 -0.0132 -0.9999>}
 spine_05 {parent=spine_04, loc=<10.0146 0 0>, rot=<0 0 -0.0015 -1>}
 neck_01 {parent=spine_05, loc=<11.3138 0.1529 0>, rot=<0 0 -0.2003 -0.9797>}
 neck_02 {parent=neck_01, loc=<4.5627 0 0>, rot=<0 0 0.0504 -0.9987>}
 head {parent=neck_02, loc=<5.3914 0 0>, rot=<0 0 0.115 -0.9934>}
 dyn_hair_bk_r_MSTR {parent=head, loc=<3.3975 -1.7145 5.7745>, rot=<0.1725 -0.985 0 0>}
 dyn_hair_bk_r_01 {parent=dyn_hair_bk_r_MSTR, loc=<0 0 0>, rot=<0 0 0 1>}
 dyn_hair_bk_r_02 {parent=dyn_hair_bk_r_01, loc=<4.6093 0.0342 0>, rot=<0 0 0.0252 -0.9997>}
 dyn_hair_bk_MSTR {parent=head, loc=<2.9118 -5.2509 0>, rot=<0.1725 -0.985 0 0>}
 dyn_hair_bk_01 {parent=dyn_hair_bk_MSTR, loc=<0 0 0>, rot=<0 0 0 1>}
 dyn_hair_bk_02 {parent=dyn_hair_bk_01, loc=<4.4635 0 0>, rot=<0 0 -0.115 -0.9934>}
 dyn_hair_bk_l_MSTR {parent=head, loc=<3.2819 -1.5204 -6.4544>, rot=<0.1725 -0.985 0 0>}
 dyn_hair_bk_l_01 {parent=dyn_hair_bk_l_MSTR, loc=<0 0 0>, rot=<0 0 0 1>}
 dyn_hair_bk_l_02 {parent=dyn_hair_bk_l_01, loc=<3.7354 -0.1709 0>, rot=<0 0 0.0086 -1>}
 dyn_hair_l_MSTR {parent=head, loc=<5.5024 6.3649 -6.3306>, rot=<-0.1492 -0.1957 0.8042 -0.541>}
 dyn_hair_l_01 {parent=dyn_hair_l_MSTR, loc=<0 0 0>, rot=<0 0 0 1>}
 dyn_hair_l_02 {parent=dyn_hair_l_01, loc=<3.5268 0.1458 -0.0404>, rot=<0.0201 -0.0132 -0.0519 -0.9984>}
 dyn_hair_r_MSTR {parent=head, loc=<6.8981 5.9076 7.0073>, rot=<0.4358 -0.8802 0.1614 0.0965>}
 dyn_hair_r_01 {parent=dyn_hair_r_MSTR, loc=<0 0 0>, rot=<0 0 0 1>}
 dyn_hair_r_02 {parent=dyn_hair_r_01, loc=<3.8495 -0.0801 0.2787>, rot=<-0.1109 0.0716 -0.0169 -0.9911>}
 dyn_hair_fr_MSTR {parent=head, loc=<10.7841 10.0324 -0.5957>, rot=<-0.1394 -0.9778 0.0234 0.155>}
 dyn_hair_fr_01 {parent=dyn_hair_fr_MSTR, loc=<0 0 0>, rot=<0 0 0 1>}
 dyn_hair_fr_02 {parent=dyn_hair_fr_01, loc=<2.7285 0 0>, rot=<-0.0046 -0.193 0.2593 -0.9463>}
 dyn_strand_fr_l {parent=head, loc=<8.3724 9.2651 -2.5702>, rot=<-0.0339 -0.9573 -0.0125 -0.2868>}
 clavicle_l {parent=spine_05, loc=<4.3575 -0.1466 -2.3223>, rot=<0.0465 -0.7289 -0.009 -0.6829>}
 clavicle_r {parent=spine_05, loc=<4.3575 -0.1466 2.3223>, rot=<0.7289 0.0465 -0.6829 0.009>}
]
