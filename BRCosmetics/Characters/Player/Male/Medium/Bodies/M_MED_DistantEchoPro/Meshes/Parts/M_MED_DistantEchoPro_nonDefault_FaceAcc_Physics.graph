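# Physics asset for the non-default FaceAcc mesh part of the male medium DistantEchoPro player body, set up for dynamic hair motion.
# Six two-bone hair chains (front, left, right, back, back-left, back-right) use capsule bodies, each root constrained to the head and each tip to its root.
# Hair joints have narrow swing and twist limits.
# Capsules on head, neck, upper spine and clavicles form the supporting body, linked by free-rotation constraints.
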PhysicsAsset /BRCosmetics/Characters/Player/Male/Medium/Bodies/M_MED_DistantEchoPro/Meshes/Parts/M_MED_DistantEchoPro_nonDefault_FaceAcc_Physics
bodies {
 clavicle_l [{
   shapeType=Capsule
   length=13.7424
   radius=10.8878
   shapeName=clavicle_l_capsule
   center=<5.6002 0.7076 -1.1125>
   rotation=<rot 72.5956 0.2867 177.6871>
  }]
 clavicle_r [{
   shapeType=Capsule
   length=14.5371
   radius=10.0932
   shapeName=clavicle_r_capsule
   center=<-5.4989 -0.3761 1.092>
   rotation=<rot 72.0363 -0.0005 176.5068>
  }]
 dyn_hair_bk_01 [{shapeType=Capsule, length=0.3849, radius=3, shapeName=dyn_hair_bk_01_capsule, center=<3 -2 0>, rotation=<rot 90 0 0>}]
 dyn_hair_bk_02 [{shapeType=Capsule, length=6.5765, radius=2.5021, shapeName=dyn_hair_bk_02_capsule, center=<2 -2 0>, rotation=<rot -0.2697 -179.9941 180>}]
 dyn_hair_bk_l_01 [{shapeType=Capsule, length=3, radius=3, shapeName=dyn_hair_bk_l_01_capsule, center=<3 -2 0>, rotation=<rot 90 0 0>}]
 dyn_hair_bk_l_02 [{shapeType=Capsule, length=3, radius=3, shapeName=dyn_hair_bk_l_02_capsule, center=<3 -2 0>, rotation=<rot 90 0 0>}]
 dyn_hair_bk_r_01 [{shapeType=Capsule, length=1.4544, radius=3, shapeName=dyn_hair_bk_r_01_capsule, center=<1.8985 -2 0>, rotation=<rot 90 0 0>}]
 dyn_hair_bk_r_02 [{shapeType=Capsule, length=2.916, radius=2.4809, shapeName=dyn_hair_bk_r_02_capsule, center=<3 -2 0>, rotation=<rot -17.0263 8.1352 -44.5441>}]
 dyn_hair_fr_01 [{shapeType=Capsule, length=3, radius=1.5, shapeName=dyn_hair_fr_01_capsule, center=<1.0141 0 0>, rotation=<rot 90 0 0>}]
 dyn_hair_fr_02 [{shapeType=Capsule, length=3, radius=1.5, shapeName=dyn_hair_fr_02_capsule, center=<1.0141 0 0>, rotation=<rot 90 0 0>}]
 dyn_hair_l_01 [{shapeType=Capsule, length=3, radius=1.5, shapeName=dyn_hair_l_01_capsule, center=<2.371 0 -0.1165>, rotation=<rot 87.1874 -0.0016 -0.0015>}]
 dyn_hair_l_02 [{shapeType=Capsule, length=3, radius=1.5, shapeName=dyn_hair_l_02_capsule, center=<2.371 0 -0.1165>, rotation=<rot 87.1874 -0.0018 -0.0018>}]
 dyn_hair_r_01 [{shapeType=Capsule, length=3, radius=1.5, shapeName=dyn_hair_r_01_capsule, center=<1.2845 0.7966 0>, rotation=<rot 90 179.4936 179.4936>}]
 dyn_hair_r_02 [{shapeType=Capsule, length=3, radius=1.5, shapeName=dyn_hair_r_02_capsule, center=<1.2845 0.7966 0>, rotation=<rot 90 180 180>}]
 head [{shapeType=Capsule, length=3.388, radius=5.1916, center=<3.1989 4.2177 0>, rotation=<rot 90 0 0>}]
 neck_01 [{shapeType=Capsule, length=8.3992, radius=5.1025, shapeName=neck_01_capsule, center=<5.3651 1.4409 0>, rotation=<rot 0 -90.8349 90>}]
 spine_05 [{shapeType=Capsule, length=0.0905, radius=12.9392, shapeName=spine_05_capsule, center=<3.2697 1.8184 0>, rotation=<rot 0 -90 0>}]
}
constraints [
 {bone1Name=dyn_hair_bk_01, bone2Name=head, swing1LimitDegrees=30, swing2LimitDegrees=20, twistLimitDegrees=2}
 {bone1Name=dyn_hair_bk_02, bone2Name=dyn_hair_bk_01, swing1LimitDegrees=15, swing2LimitDegrees=10, twistLimitDegrees=5}
 {bone1Name=dyn_hair_bk_l_01, bone2Name=head, swing1LimitDegrees=25, swing2LimitDegrees=15, twistLimitDegrees=2}
 {bone1Name=dyn_hair_bk_l_02, bone2Name=dyn_hair_bk_l_01, swing1LimitDegrees=15, swing2LimitDegrees=10, twistLimitDegrees=5}
 {bone1Name=dyn_hair_bk_r_01, bone2Name=head, swing1LimitDegrees=20, swing2LimitDegrees=20, twistLimitDegrees=2}
 {bone1Name=dyn_hair_bk_r_02, bone2Name=dyn_hair_bk_r_01, swing1LimitDegrees=30, swing2LimitDegrees=30, twistLimitDegrees=5}
 {
  bone1Name=head
  bone2Name=neck_01
  swing1LimitDegrees=45
  swing1Motion=Free
  swing2LimitDegrees=45
  swing2Motion=Free
  twistLimitDegrees=45
  twistMotion=Free
 }
 {bone1Name=dyn_hair_l_01, bone2Name=head, swing1LimitDegrees=20, swing2LimitDegrees=15, twistLimitDegrees=1}
 {bone1Name=dyn_hair_l_02, bone2Name=dyn_hair_l_01, swing1LimitDegrees=20, swing2LimitDegrees=15, twistLimitDegrees=1}
 {bone1Name=dyn_hair_fr_02, bone2Name=dyn_hair_fr_01, swing1LimitDegrees=20, swing2LimitDegrees=20, twistLimitDegrees=5}
 {bone1Name=dyn_hair_fr_01, bone2Name=head, swing1LimitDegrees=20, swing2LimitDegrees=20, twistLimitDegrees=5}
 {bone1Name=dyn_hair_r_02, bone2Name=dyn_hair_r_01, swing1LimitDegrees=20, swing2LimitDegrees=15, twistLimitDegrees=1}
 {bone1Name=dyn_hair_r_01, bone2Name=head, swing1LimitDegrees=20, swing2LimitDegrees=15, twistLimitDegrees=1}
 {
  bone1Name=neck_01
  bone2Name=spine_05
  swing1LimitDegrees=45
  swing1Motion=Free
  swing2LimitDegrees=45
  swing2Motion=Free
  twistLimitDegrees=45
  twistMotion=Free
 }
 {
  bone1Name=clavicle_l
  bone2Name=spine_05
  swing1LimitDegrees=45
  swing1Motion=Free
  swing2LimitDegrees=45
  swing2Motion=Free
  twistLimitDegrees=45
  twistMotion=Free
 }
 {
  bone1Name=clavicle_r
  bone2Name=spine_05
  swing1LimitDegrees=45
  swing1Motion=Free
  swing2LimitDegrees=45
  swing2Motion=Free
  twistLimitDegrees=45
  twistMotion=Free
 }
]
