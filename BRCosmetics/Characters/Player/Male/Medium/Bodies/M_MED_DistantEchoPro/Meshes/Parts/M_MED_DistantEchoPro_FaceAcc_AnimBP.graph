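# Animation Blueprint for the DistantEchoPro face accessory, copying the attached parent pose and adding physics with per-variant handling.
# The default Stage1 variant runs head-based RigidBody physics into a Control Rig; other variants use spine_05-based physics with a separate asset, then additive back-hair bone rotations.
# The event graph detects the variant, updates gravity override, and derives hair correction alpha from head-to-neck rotation, reduced when not targeting.
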AnimBlueprint /BRCosmetics/Characters/Player/Male/Medium/Bodies/M_MED_DistantEchoPro/Meshes/Parts/M_MED_DistantEchoPro_FaceAcc_AnimBP
target_skeleton=&/M_MED_DistantEchoPro_FaceAcc_Skeleton
graphs [
 AnimGraph {nodes [
   n1 = RigidBody!(
    ComponentPose=Local To Component(LocalPose=Copy Pose From Mesh() {bUseAttachedParent=true})
    Alpha
    OverrideWorldGravity:Vector=Get Gravity Override()
    ExternalForce
   ) -> (Pose) {
    bOverrideWorldGravity=true
    overridePhysicsAsset=&/M_MED_DistantEchoPro_nonDefault_FaceAcc_Physics
    simulationSpace=BaseBoneSpace
    alphaBoolBlend {blendOption=Linear}
    baseBoneRef {boneName=spine_05}
    componentAppliedLinearAccClamp=<3000 3000 3000>
    componentLinearAccScale=<0.3 0.3 0.3>
    componentLinearVelScale=<0.3 0.3 0.3>
    overrideWorldGravity {z=-980}
   }
   n2 = Get head Rot ADS Up() -> (head Rot ADS Up:float)
   n3 = "Transform (Modify) Bone - Bone: dyn_hair_bk_01"(ComponentPose=@n1, Alpha=@n2, Translation="0,0,0", Rotation="0,-25,27", Scale="1,1,1") {
    rotationMode=BMM_Additive
    rotationSpace=BCS_BoneSpace
    alphaBoolBlend {blendOption=Linear}
    boneToModify {boneName=dyn_hair_bk_01}
    rotation {roll=27, yaw=-25}
   }
   n4 = "Transform (Modify) Bone - Bone: dyn_hair_bk_02"(ComponentPose=@n3, Alpha=@n2, Translation="0,0,0", Rotation="0,-15,0", Scale="1,1,1") {
    rotationMode=BMM_Additive
    rotationSpace=BCS_BoneSpace
    alphaBoolBlend {blendOption=Linear}
    boneToModify {boneName=dyn_hair_bk_02}
    rotation {yaw=-15}
   }
   n5 = "Transform (Modify) Bone - Bone: dyn_hair_bk_r_01"(ComponentPose=@n4, Alpha=@n2, Translation="0,0,0", Rotation="0,-40,0", Scale="1,1,1") {
    rotationMode=BMM_Additive
    rotationSpace=BCS_BoneSpace
    alphaBoolBlend {blendOption=Linear}
    boneToModify {boneName=dyn_hair_bk_r_01}
    rotation {yaw=-40}
   }
   n6 = "Transform (Modify) Bone - Bone: dyn_hair_bk_r_02"(ComponentPose=@n5, Alpha=@n2, Translation="0,0,0", Rotation="0,-10,0", Scale="1,1,1") {
    rotationMode=BMM_Additive
    rotationSpace=BCS_BoneSpace
    alphaBoolBlend {blendOption=Linear}
    boneToModify {boneName=dyn_hair_bk_r_02}
    rotation {yaw=-10}
   }
   n7 = RigidBody(ComponentPose=Local To Component(LocalPose=Copy Pose From Mesh() {bUseAttachedParent=true}), OverrideWorldGravity=Get Gravity Override()) {
    bOverrideWorldGravity=true
    overridePhysicsAsset=&/M_MED_DistantEchoPro_FaceAcc_Physics
    simulationSpace=BaseBoneSpace
    alphaBoolBlend {blendOption=Linear}
    baseBoneRef {boneName=head}
    componentAppliedLinearAccClamp=<3000 3000 3000>
    componentLinearAccScale=<0.3 0.3 0.35>
    componentLinearVelScale=<0.3 0.3 0.35>
    overrideWorldGravity {z=-980}
   }
   n8 = Control Rig(Source=Component To Local(ComponentPose=@n7)) {
    alphaBoolBlend {blendOption=Linear}
    controlRigAssetReference {blueprintRigClass=&/M_MED_DistantEchoPro_FaceAcc_CtrlRig.M_MED_DistantEchoPro_FaceAcc_CtrlRig_C}
   }
   Output Pose(Result=Blend Poses by bool(BlendPose_0=@n8, BlendPose_1=Component To Local(ComponentPose=@n6), bActiveValue=Get is default variant() -> (is default variant:bool)) {blendType=Linear}) {layerGroup=None}
  ]}
 EventGraph {nodes [
   Event BlueprintUpdateAnimation()
   > n1 = Sequence!(execute) -> (then_0, then_1, then_2:Exec, then_3:Exec)
   > n2 = Cast To FortPlayerPawn(Object=GetOwningActor())
   > n3 = GetVariantChannelActiveIndexForItemActor!(
    execute
    Item:Athena Cosmetic Item Definition Object Reference=/BRCosmetics/Athena/Items/Cosmetics/Characters/Character_DistantEchoPro
    Actor:Actor Object Reference=@n2
    Channel:Gameplay Tag Structure=(TagName="Cosmetics.Variant.Channel.Parts")
    VariantList:Array of Gameplay Tag Structures=Make Array("[0]"=(TagName="Cosmetics.Variant.Property.Stage1"))
    bErrorIfChannelMissing:bool=false
    bErrorIfTagNotInArray:bool=false
   ) -> (then, Index:int=0, bIndexIsValid:bool, bFoundChannel:bool)
   > n4 = Branch(Condition=AND Boolean(A="Equal (Integer)"(A=@n3), B=@n3.bIndexIsValid))
   > Set is default variant(is default variant:bool=true) -> (Output_Get:bool)
   Set is default variant(execute=@n4.else, is default variant=false)
   Set BodyMesh(execute=@n1.then_1, BodyMesh=GetPartSkeletalMeshComponent())
   n5 = Set head Rot ADS Up(
    execute=@n1.then_2
    head Rot ADS Up:float=GetAlphaFromReferenceTransformDelta(
     SkeletalMeshComponent=Get BodyMesh() -> (BodyMesh:Skeletal Mesh Component Object Reference)
     BoneName=head
     BaseName=neck_02
     InputRange=(X=-33,Y=0)
     OutputRange=(X=1,Y=0)
     TwistAxis="1,0,0"
    )
   ) -> (Output_Get:float)
   > n6 = Branch(Condition=Get bIsTargeting())
   > Set head Rot ADS Up(head Rot ADS Up=@n5)
   Set head Rot ADS Up(execute=@n6.else, head Rot ADS Up=float * float(A=@n5, B=0.8) {is=title K2Node_PromotableOperator})
   n7 = Gravity Override State Function DEPRECATED(execute=@n1.then_3, CharacterPartAnimInstance=Self-Reference(), GameTimeInSeconds=0, GravityOverrideParams=Get Gravity Override Params())
   > Set Gravity Override(Gravity Override=@n7)
  ]}
]
variables [
 BodyMesh {type=SkeletalMeshComponent}
 Gravity Override {type=Vector}
 Gravity Override Params {type=GravityOverrideParamsStruct, default {skydive_FloatingDown {gravityOverride=<0 -300 300>}}}
 head Rot ADS Up {type=number}
 is default variant {type=boolean}
]
